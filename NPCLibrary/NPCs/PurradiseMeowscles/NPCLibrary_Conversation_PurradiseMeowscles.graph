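# Conversation database for the Purradise Meowscles NPC, giving its greeting speech and menu links when a player talks to it.
# The entry point leads to one speech node with a default line plus contextual lines for a first-ever conversation and for players wearing specific outfits, ranked by curve-based priorities.
# The speech node then branches through link nodes to the shared NPC quest and basic services conversations.
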ConversationDatabase /NPCLibrary/NPCs/PurradiseMeowscles/NPCLibrary_Conversation_PurradiseMeowscles
entry_points [{tag=Conversation.Entry.PurradiseMeowscles, destinations [#1]}]
nodes [
 ConversationEntryPointNode_7 {
  id=#1
  class=ConversationEntryPointNode
  outputs [#2]
  properties {entryTag {tagName=Conversation.Entry.PurradiseMeowscles}}
 }
 Speech_C_4 {
  id=#2
  class=Speech_C
  outputs [#3, #4]
  properties {
   nodeName=""
   generalConfig {
    defaultMessage="Another day in purradise, mrow!"
    contextualMessages [
     {
      message=The new coif is worth the extra hairballs.
      contextRequirements [{requirement {class=FortControllerRequirement_IsFirstEverConversationWithNPC}}]
      priority {curve {rowName=Tandem.Speech.Priority.FirstEverInteraction}}
     }
     {
      message="Don't cannonball in your mecha-suit, mrow!"
      contextRequirements [{requirement {
         class=FortControllerRequirement_HasCID
         properties {allowedCIds [/BRCosmetics/Athena/Items/Cosmetics/Characters/CID_776_Athena_Commando_M_ProfessorPup]}
        }}]
      priority {curve {rowName=Tandem.Speech.Priority.OutfitReaction}}
     }
     {
      message="So, uh... You like the hair?"
      contextRequirements [{requirement {
         class=FortControllerRequirement_HasCID
         properties {allowedCIds [/BRCosmetics/Athena/Items/Cosmetics/Characters/CID_286_Athena_Commando_F_NeonCat]}
        }}]
      priority {curve {rowName=Tandem.Speech.Priority.OutfitReaction}}
     }
     {
      message=Lemme guess. Your skateboard got stuck in the mud again.
      contextRequirements [{requirement {class=FortControllerRequirement_HasCID, properties {allowedCIds [&Character_ChillCat]}}}]
      priority {curve {rowName=Tandem.Speech.Priority.OutfitReaction}}
     }
     {
      message=Mroww. All I want is to curl up in a spot of sun.
      contextRequirements [{requirement {
         class=FortControllerRequirement_HasCID
         properties {allowedCIds [&CID_814_Athena_Commando_M_BananaSummer]}
        }}]
      priority {curve {rowName=Tandem.Speech.Priority.OutfitReaction}}
     }
     {
      message=No covert ops. Just covert naps.
      contextRequirements [{requirement {
         class=FortControllerRequirement_HasCID
         properties {allowedCIds [/BRCosmetics/Athena/Items/Cosmetics/Characters/CID_A_129_Athena_Commando_M_CatBurglarSummer]}
        }}]
      priority {curve {rowName=Tandem.Speech.Priority.OutfitReaction}}
     }
    ]
   }
  }
 }
 ConversationLinkNode_0 {id=#4, class=ConversationLinkNode, properties {remoteEntryTag {tagName=Conversation.Entry.BasicServices}}}
 ConversationLinkNode_1 {id=#3, class=ConversationLinkNode, properties {remoteEntryTag {tagName=Conversation.Entry.NPCLibrary.Quests}}}
]
